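# Sound cue that plays when picking up the Puppeteer boss medallion in the FirePetal boons content.
# A mixer layers the default takedown medallion pickup sound with a louder boss stinger, which starts after a short fixed delay.
# The cue plays in the loot collect sound class and overrides concurrency to allow a single instance at a time.
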SoundCue /FirePetalBoonsGameplay/Assets/Audio/S33_PuppeteerBoss_Medallion_Pickup_Cue
root=Mixer_0
nodes [
 {id=Delay_1, children [WavePlayer_0], properties {delayMax=0.2, delayMin=0.2}}
 {id=Mixer_0, children [WavePlayer_1,Delay_1], properties {inputVolume [1,1.45]}}
 {id=WavePlayer_0, properties {soundWaveAssetPtr=&/S33_PuppeteerBoss_Medallion_Stinger}}
 {id=WavePlayer_1, properties {soundWaveAssetPtr=&/Default_TakedownMedallions_Pickup_SFX}}
]
properties {
 bOverrideConcurrency=true
 maxDistance=2097152
 soundClassObject=&Fort_Loot_Collect
 volumeMultiplier=0.7692
 concurrencyOverrides {maxCount=1}
}
