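# SoundCue for the Fire Petal Boons wind sound, variant B, played through the loot-collect sound class at reduced volume.
# A single wave player feeds a Modulator that varies pitch and volume on each play.
# It uses a close weapon attenuation asset and overrides concurrency so only one instance plays at a time.
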SoundCue /FirePetalBoonsGameplay/Assets/Audio/SC_FirePetalBoons_Wind_B
root=Modulator_0
nodes [
 {id=Modulator_0, children [WavePlayer_0], properties {pitchMax=1.25, volumeMax=1, volumeMin=0.7}}
 {id=WavePlayer_0, properties {soundWaveAssetPtr=&/FirePetalBoons_wind_B}}
]
properties {
 attenuationSettings=/Game/Sounds/Attenuation/Fort_Weapon_AB_Close_Athena_Att
 bOverrideConcurrency=true
 maxDistance=4500
 soundClassObject=&Fort_Loot_Collect
 volumeMultiplier=0.2
 concurrencyOverrides {maxCount=1}
}
