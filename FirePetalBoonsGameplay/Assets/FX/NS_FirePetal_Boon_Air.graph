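# Niagara effect for the Fire Petal air boon: a soft airflow of mesh particles bursting upward.
# Each three-particle burst varies lifetime, scale and height offset by spawn index for a stacked look.
# Meshes rise against drag, spin around Z with a slowing rate, and narrow and stretch vertically while alpha fades.
# Time-of-day values drive color tint and a material parameter, while a curve animates another parameter.
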NiagaraSystem /FirePetalBoonsGameplay/Assets/FX/NS_FirePetal_Boon_Air
dependencies {usedDataInterfaces [&NiagaraDataInterfaceCurve], usedDynamicInputs [], usedModules [&SystemState]}
emitters {NS_SoftAirflow {
  emitterUpdateScript [
   &EmitterState()
   &SpawnBurst_Instantaneous(Spawn Count=3, Use Spawn Probability=true, UseLoopCountLimit=false)
  ]
  particleSpawnScript [
   /SetVariables(Particles.Random=&FloatFromCurve(FloatCurve=DataInterface() {propertyValues {Curve {keys [{tangentMode=RCTM_User, time=0, value=0}, {tangentMode=RCTM_User, time=1, value=1}]}}})) {isSetParameters=true}
   &InitializeParticle(
    Color=<rgba 0.8698 0.8698 0.8698 1>
    Lifetime=&Lerp_Float(A=0.66, B=1, Alpha=&ReturnNormalizedExecIndex())
    Mesh Scale=&Multiply_VectorByFloat(
     Vector=<2 2 2>
     Float=&Multiply_Float(
      A=1
      B=&Lerp_Float(
       A=1
       B=0.5
       Alpha=&FloatFromCurve(
        FloatCurve=DataInterface() {propertyValues {Curve {keys [
            {arriveTangent=1.7769, interpMode=RCIM_Cubic, leaveTangent=1.7769, tangentMode=RCTM_User, time=0, value=0}
            {interpMode=RCIM_Cubic, leaveTangent=0.5389, time=1, value=1}
           ]}}}
        CurveIndex=&ReturnNormalizedExecIndex()
       )
      )
     )
    )
    Mesh Scale Mode=::NewEnumerator4
    Position Offset=&Multiply_VectorByFloat(Vector=<0 0 40>, Float=&ReturnNormalizedExecIndex())
    Position Offset Coordinate Space=::Local
    UsePositionOffset=true
    Write Material Random=true
   )
   &InitialMeshOrientation(
    Mesh Orientation Mode=::NewEnumerator7
    Rotation=&RandomRangeVector(Minimum=<0 0 -1>, Maximum=<0 0 1>)
    Use Rotation Vector=true
   )
   &AddVelocity(Velocity=<0 0 1>, Velocity Speed Scale=150)
  ]
  particleUpdateScript [
   &ParticleState()
   &Drag(Use Linear Drag=true, Use Rotational Drag=true)
   &SolveForcesAndVelocity(Clamp Velocity=false, Limit Acceleration=false)
   &DynamicMaterialParameters(
    Index 0 Param 1=&FloatFromCurve(FloatCurve=DataInterface() {propertyValues {Curve {keys [{interpMode=RCIM_Cubic, tangentMode=RCTM_User, time=0.5, value=0}, {interpMode=RCIM_Cubic, tangentMode=RCTM_User, time=1, value=1}]}}})
    Index 0 Param 2=&FNTimeOfDayScalar(Night Scale=0, Morning Scale=1, Evening Scale=1)
    Index 0 Param 3=1
    Module.Param0WriteEnabled=true
    Module.Param1WriteEnabled=true
    Module.Param2WriteEnabled=true
    Module.Param3WriteEnabled=true
   )
   &ScaleMeshSize(Scale Factor=&Multiply_VectorByFloat(
     Vector=&Lerp_Vector(
      A=<0.25 0.25 1.5>
      B=<1 1 1>
      Alpha=&FloatFromCurve(FloatCurve=DataInterface() {propertyValues {Curve {keys [{arriveTangent=-1.4221, interpMode=RCIM_Cubic, leaveTangent=-1.4221, tangentMode=RCTM_User, time=0, value=1}, <at 1 0 RCIM_Cubic>]}}})
     )
     Float=1
    ))
   &UpdateMeshOrientation(Rotation Rate=&Lerp_Float(A=1.5, B=0.5, Alpha=$Particles.NormalizedAge), Rotation Vector=<0 0 1>)
   &ScaleColor(
    Scale Alpha=&FloatFromCurve(
     FloatCurve=DataInterface() {propertyValues {Curve {keys [
         {arriveTangent=1.8077, interpMode=RCIM_Cubic, time=0, value=0}
         <at 0.66 0.5 RCIM_Cubic>
         {arriveTangent=-0.7636, interpMode=RCIM_Cubic, leaveTangent=-0.7636, tangentMode=RCTM_User, time=1, value=0}
        ]}}}
     Scale Curve=0.7
    )
    Scale RGB=&MakeVectorFromLinearColorRGB(Color=&FNTimeOfDayColor(Night Color=<rgba 0.5088 0.6003 0.9 1>, Morning Color=<rgba 0.7 0.7 0.7 1>, Day Color=<rgba 1 1 1 1>, Evening Color=<rgba 0.7 0.7 0.7 1>))
    ScaleA=true
    ScaleRGB=true
   )
  ]
  rendererClasses [&NiagaraMeshRendererProperties]
  renderers [{rendererClass=&NiagaraMeshRendererProperties, rendererIndex=0}]
 }}
summary {
 systemName=NS_FirePetal_Boon_Air
 emitters [{emitterName=NS_SoftAirflow, rendererClasses [&NiagaraMeshRendererProperties]}]
 userVariables [
  User.bIsFrontEnd {type=NiagaraBool}
  User.bIsFrontEndPreview {type=NiagaraBool}
  User.RemoveAirflow {type=NiagaraBool}
 ]
}
