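# Parent additive, unlit skeletal-mesh material for the Void Run cloak effect, with cracked-glass detail, fresnel rim, fill and glitch noise.
# Timestamps drive a quantized-fresnel transition and expiry glow, while a cloak MPC opacity value blends stationary and moving looks.
# Crack UVs use pre-skinned position with bump offset, and scene-color luminance adjusts brightness off mobile.
# Refraction is dithered and distance-faded, a custom-depth comparison masks hidden layers, and WPO hides or collapses vertices.
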Material /FirePetalBoonsGameplay/Assets/Materials/M_VoidRun_Cloak_Parent
graph {
 properties {
  allowTranslucentCustomDepthWrites=true
  bEnableResponsiveAA=true
  bUsedWithClothing=true
  bUsedWithMeshDeformer=true
  bUsedWithMorphTargets=true
  bUsedWithSkeletalMesh=true
  blendMode=BLEND_Additive
  pixelDepthOffsetMode=PDOM_Legacy
  refractionMethod=RM_PixelNormalOffset
  shadingModel=MSM_Unlit
  usedShadingModels=Unlit
 }
 nodes [
  n1 = ObjectPositionWS()
  n2 = WorldPosition()
  Add(
   A=MaterialFunctionCall!(WPO) {materialFunction=/Game/Packages/Fortress_SharedMaterials/Base_Material_Functions/Fortnite_HideElementsWithVertexColor}.Result
   B=Multiply(A=Subtract(A=@n1, B=@n2.XYZ), B=ScalarParameter() {parameterName=HT_CrunchVerts})
  )
  Composite() {subgraphName=LuminanceFactors}
  n3 = VertexInterpolator(VS=Divide(A=PreSkinnedPosition(), B=Constant() {r=200}))
  n4 = ScalarParameter() {defaultValue=1, parameterName=CracksTiling}
  n5 = BumpOffset(
   Coordinate=ComponentMask(None=Multiply(A=@n3.PS, B=@n4) {constB=-1}) {b=true, r=true}
   Height=Constant() {r=1}
   HeightRatioInput=ScalarParameter() {defaultValue=-0.3, parameterName=CracksHeightRatio}
  )
  n6 = Power(Base=TextureSample(UVs=NamedRerouteDeclaration(None=@n5) {name=Crack UV, nodeColor=<rgba 0 0.0625 1 1>}) {texture=/Game/Effects/Fort_Effects/Textures/Tile/T_Glass_Cracks_M}.G) {constExponent=1.5}
  n7 = VectorParameter() {parameterName=EmissiveColor, defaultValue=<rgba 0.6282 0.9926 5 1>}
  n8 = Multiply(A=Multiply(A=@n6, B=ScalarParameter() {defaultValue=4, parameterName=CracksFadeInOpacity}), B=@n7.RGB)
  n9 = ScalarParameter() {defaultValue=0.5, parameterName=ExpireTransitionLength}
  n10 = MaterialFunctionCall(
   TimestampA=ScalarParameter() {parameterName=ExpireTimeStampA}
   TransitionLength_In=@n9
   TimestampB=ScalarParameter() {parameterName=ExpireTimeStampB}
   TransitionLength_Out=@n9
   Transition In Only?=StaticBool()
  ) {materialFunction=&MF_Timestamp_InOut_01}
  n11 = LinearInterpolate(
   A=CollectionParameter() {collection=/CloakGameplay/Materials/MPC_Cloak_Generic, parameterName=Opacity}
   B=ScalarParameter() {defaultValue=1.3, parameterName=ExpireVisibility}
   Alpha=@n10.SmoothStep
  )
  n12 = LinearInterpolate(
   A=ScalarParameter() {defaultValue=3, parameterName=FresnelExponentStationary}
   B=ScalarParameter() {defaultValue=0.4, parameterName=FresnelExponentMoving}
   Alpha=@n11
  )
  n13 = ScalarParameter() {defaultValue=0.5, parameterName=TransitionLength}
  n14 = MaterialFunctionCall(
   TimestampA=ScalarParameter() {parameterName=TimeStampA}
   TransitionLength_In=@n13
   TimestampB=ScalarParameter() {defaultValue=10000000000, parameterName=TimeStampB}
   TransitionLength_Out=@n13
   Transition In Only?=StaticBool()
  ) {materialFunction=&MF_Timestamp_InOut_01}
  n15 = Saturate(None=Add(A=@n14.Linear) {constB=0.12})
  n16 = Multiply(
   A=TextureSample(UVs=NamedRerouteUsage()) {samplerType=SAMPLERTYPE_Normal, texture=/Game/Effects/Fort_Effects/Textures/Tile/T_glass_cracks_sharp_N}.RGB
   B=VectorParameter() {parameterName=DetailNormalIntensity, defaultValue=<rgba 0.5 0.5 1 0>}.RGB
  )
  n17 = ScalarParameter() {defaultValue=2, parameterName=QuantizeSteps}
  n18 = Multiply(
   A=MaterialFunctionCall(Exponent=Multiply(A=@n12, B=@n15), Vertex / Pixel Normal=Transform(None=@n16)) {materialFunction=&MF_Fresnel_FX}.Fresnel Inverted
   B=@n17
  )
  n19 = Multiply(
   A=LinearInterpolate(A=VectorParameter() {parameterName=EmissiveColor_FresnelBlend, defaultValue=<rgba 0.6282 0.9926 5 1>}.RGB, B=@n7.RGB, Alpha=@n18)
   B=@n18
  )
  n20 = LinearInterpolate(
   A=ScalarParameter() {parameterName=CracksOpacityStationary}
   B=ScalarParameter() {defaultValue=1, parameterName=CracksOpacityMoving}
   Alpha=NamedRerouteUsage()
  )
  n21 = Multiply(A=VectorParameter() {parameterName=CracksColor, defaultValue=<rgba 0.1 0.0202 0 1>}.RGB, B=Multiply(A=@n6, B=@n20))
  n22 = VectorParameter() {parameterName=BaseFresnelColor_High, defaultValue=<rgba 0 0 0 1>}
  n23 = VectorParameter() {parameterName=BaseFresnelColor, defaultValue=<rgba 0 0 0 1>}
  n24 = Saturate(None=DotProduct(A=CameraVectorWS(), B=VertexNormalWS()))
  n25 = NamedRerouteDeclaration(None=OneMinus(None=@n24)) {name=FresnelBase, nodeColor=<rgba 0.8047 1 0 1>}
  n26 = ShadingPathSwitch(
   Default=ScalarParameter() {defaultValue=2, parameterName=BaseFresnelExponent_High}
   Mobile=ScalarParameter() {defaultValue=4, parameterName=BaseFresnelExponent_Low}
  ) {desc=Migrated from FeatureLevelSwitch}
  n27 = NamedRerouteUsage()
  n28 = Multiply(
   A=Multiply(
    A=ShadingPathSwitch(Default=@n22.RGB, Mobile=@n23.RGB) {desc=Migrated from FeatureLevelSwitch}
    B=Power(Base=@n25, Exp=@n26) {constExponent=8}
   )
   B=LinearInterpolate(A=ScalarParameter() {parameterName=BaseFresnelStationary}, B=ScalarParameter() {defaultValue=1, parameterName=BaseFresnelMoving}, Alpha=@n27)
  )
  n29 = ScalarParameter() {defaultValue=1, parameterName=FillExponent}
  n30 = Multiply(
   A=Multiply(A=Power(Base=@n24, Exp=@n29), B=VectorParameter() {parameterName=FillColor, defaultValue=<rgba 0 0 0 1>}.RGB)
   B=LinearInterpolate(A=ScalarParameter() {parameterName=FillOpacityStationary}, B=ScalarParameter() {defaultValue=1, parameterName=FillOpacityMoving}, Alpha=@n27)
  )
  n31 = Add(A=NamedRerouteUsage(), B=Add(A=Multiply(A=@n21, B=@n20), B=Add(A=@n28, B=@n30) {constB=0}))
  n32 = Constant3Vector() {constant=<rgba 0.2722 0.6741 0.0537 1>}
  MaterialExpressionReroute_2 = Reroute(None=@n32)
  n33 = DotProduct(A=@MaterialExpressionReroute_2.Output, B=SceneColor!(UVs).RGB)
  n34 = ScalarParameter() {defaultValue=2, group=LuminanceCorrection, parameterName=LuminanceCorrectionPower}
  n35 = ScalarParameter() {defaultValue=8, group=LuminanceCorrection, parameterName=LuminanceCorrectionIntensityStationary}
  n36 = ScalarParameter() {defaultValue=4, group=LuminanceCorrection, parameterName=LuminanceCorrectionIntensityMoving}
  n37 = NamedRerouteUsage()
  n38 = Add(A=Multiply(A=Power(Base=@n33, Exp=@n34), B=LinearInterpolate(A=@n35, B=@n36, Alpha=@n37)) {constB=4})
  n39 = ScalarParameter() {defaultValue=1, parameterName=OverallOpacity}
  n40 = ShadingPathSwitch(
   Default=StaticSwitchParameter(True=Multiply(A=@n38, B=@n39), False=@n39) {defaultValue=true, group=LuminanceCorrection, parameterName=UseLuminanceCorrection, sortPriority=0}
   Mobile=@n39
  ) {desc=Migrated from FeatureLevelSwitch}
  n41 = Multiply(A=@n10.SmoothStep, B=ScalarParameter() {defaultValue=4, parameterName=ExpireGlowIntensity}) {constB=4}
  n42 = OneMinus(None=Divide(A=Floor(None=@n18), B=@n17))
  n43 = ConstantBiasScale(None=@n14.Linear) {bias=-0.5, scale=2}
  n44 = LinearInterpolate(
   A=Multiply(A=Add(A=@n8, B=@n19), B=OneMinus(None=@n14.Linear))
   B=Multiply(A=Multiply(A=@n31, B=@n40), B=Add(A=@n41))
   Alpha=Saturate(None=Add(A=@n42, B=@n43))
  )
  n45 = Constant() {r=1}
  n46 = MaterialFunctionCall(Value_Morning=@n45, Value_Day=@n45, Value_Evening=@n45, Value_Night=ScalarParameter() {defaultValue=1, parameterName=TOD_NightBoost}) {materialFunction=&Scalar_Time_of_Day}
  EyeAdaptationInverse(
   LightValueInput=ShadingPathSwitch(Default=Multiply(A=@n44, B=@n46.Result), Mobile=@n44) {desc=Migrated from FeatureLevelSwitch}
   AlphaInput=ScalarParameter() {defaultValue=1.6, parameterName=TODBalance}
  )
  n47 = LinearInterpolate(
   A=ScalarParameter() {defaultValue=0.995, parameterName=RefractionStationary}
   B=ScalarParameter() {defaultValue=0.98, parameterName=RefractionMoving}
   Alpha=@n11
  )
  n48 = Distance(A=ActorPositionWS(), B=CameraPositionWS())
  n49 = ScalarParameter() {defaultValue=800, parameterName=RefractionDistanceDivisor}
  n50 = OneMinus(None=Saturate(None=Divide(A=@n48, B=@n49)))
  n51 = LinearInterpolate(
   A=ScalarParameter() {defaultValue=1.2, parameterName=RefractionTransition}
   B=LinearInterpolate(B=@n47, Alpha=@n50) {constA=1}
   Alpha=@n14.Linear
  )
  n52 = Abs(None=Divide(A=Subtract(A=@n51)))
  n53 = ScalarParameter() {defaultValue=0.05, parameterName=DitherScale}
  n54 = ScalarParameter() {defaultValue=1, parameterName=Visibility}
  n55 = LinearInterpolate(
   A=Add(A=@n52)
   B=Subtract(B=@n52)
   Alpha=Multiply(
    A=MaterialFunctionCall() {materialFunction=&DitherTemporalAA}.Result
    B=Multiply(A=@n53, B=@n54)
   )
  ) {constA=0.9, constB=1.1}
  n56 = SceneTexture() {sceneTextureId=PPI_CustomDepth}
  n57 = Constant()
  n58 = If(
   A=PixelDepth().R
   B=Add(A=ComponentMask(None=@n56.Color) {r=true}, B=ScalarParameter() {defaultValue=0.1, parameterName=cdoffset}) {constB=0.1}
   "A > B"=@n57
   "A == B"=@n57
   "A < B"=Constant() {r=1}
  )
  LinearInterpolate(B=ShadingPathSwitch(Default=Constant() {r=1}, Deferred=Multiply(A=@n51, B=@n55)), Alpha=@n58) {constA=1}
  n59 = TextureSample(UVs=NamedRerouteUsage()) {samplerType=SAMPLERTYPE_Normal, texture=&T_glass_cracks_sharp_N}
  Multiply(A=@n59.RGB, B=VectorParameter() {parameterName=DetailNormalIntensity, defaultValue=<rgba 0.5 0.5 1 0>}.RGB)
  Multiply(A=@n47, B=@n50)
  NamedRerouteDeclaration(None=@n11) {name=Motion, nodeColor=<rgba 1 0 0.6094 1>}
  n60 = MaterialFunctionCall!(Camera Relative Position, X Camera to World Vector, Y Camera to World Vector, Z Camera to World Vector, "Tan(FOV/2)*[1,Screen Res Y/X]") {materialFunction=/Engine/Functions/Engine_MaterialFunctions02/Math/ManualWorldToScreenUVsTransform}
  n61 = Multiply(A=Time(), B=ScalarParameter() {defaultValue=1.2, parameterName=GlitchNoiseSpeed})
  n62 = Multiply(
   A=SphereMask(
    A=TextureSample(UVs=@n60.Screen UVs) {samplerType=SAMPLERTYPE_LinearGrayscale, texture=/Game/Effects/Fort_Effects/Emotes/UltraLight/Texture/T_Cloud_Greyscale}.R
    B=Frac(None=@n61)
    Radius=ScalarParameter() {defaultValue=0.1, parameterName=GlitchNoiseThickness}
   ) {hardnessPercent=0}
   B=VectorParameter() {parameterName=GlitchNoiseColor, defaultValue=<rgba 0.0578 0 0.5625 1>}.RGB
  )
  n63 = Fresnel()
  NamedRerouteDeclaration(None=ShadingPathSwitch(Default=LinearInterpolate(B=@n62, Alpha=@n63) {constB=0}, Mobile=Constant()) {desc=Migrated from FeatureLevelSwitch}) {name=GlitchNoise, nodeColor=<rgba 1 0.0469 0 1>}
  PinBase()
  PinBase!(Output=@n32) {reroutePins [Output {expression=&self:MaterialExpressionReroute_2}]}
  ScalarParameter() {defaultValue=1, parameterName=DebugMovement}
  ScalarParameter() {defaultValue=1, parameterName=Visibility}
  n64 = VertexNormalWS()
  n65 = ScalarParameter() {defaultValue=2, parameterName=QuantizeSteps}
  n66 = Multiply(A=@n64, B=@n65)
  ShadingPathSwitch(Default=@n64, Deferred=Divide(A=Floor(None=@n66), B=@n65))
  n67 = Constant() {r=1}
  StaticSwitchParameter(True=QualitySwitch(Default=@n67, High=@n58, Epic=@n58), False=@n67) {defaultValue=true, parameterName=EnableBackFaceCulling}
  Transform(None=@n59.RGB)
 ]
}
